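; Vehicle mod box spawner that places a random first box after game state readiness and a staggered delay, on the server when mods are enabled.
; A box's death starts a hotfixable respawn timer and holographic countdown, with replicated visibility.
; Respawns pick mods randomly or cycle through the array.
; Before respawning, an overlap check launches player pawns upward or applies the mod directly to a vehicle and restarts the timer instead.
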(event EventBeginPlay
  (bind _modstospawn (Variables|ModBox|GetModsToSpawn))
  (bind (_outitem _outindex) (Utilities|Array|RandomArrayItem _modstospawn))
  ;[1 1 1] EVENT--> BeginPlay
  (Class|FortAsyncActionGameStateReady|GameStateReadyAsync
    (:OnReady
      (Utilities|FlowControl|Delay (Math|Random|RandomFloatinRange 0.1 1.0)) ; Stagger Mod Box Spawning at start of match
      (CallFunction|SetHotfixValues)
      (Utilities|FlowControl|DoOnce false
        (:Completed
          ;[1 1 1] Spawn First ModBox
          (Utilities|FlowControl|Sequence
            (:then_0
              (if (Networking|IsServer)
                (Utilities|CheckModsEnabled
                  (:Return
                    (Utilities|CheckModsEnabled
                      (:Return
                        (Variables|Default|SetMaxModArrayCount (Utilities|Array|Length _modstospawn))
                        (Variables|Default|SetModSpawnIndex _outindex)
                        (bind _returnvalue (Game|SpawnActorfromClass _outitem (Transformation|GetActorTransform) "AlwaysSpawn" "SelectDefaultAtRuntime" self 0 0))
                        (bind _output_get (Variables|Default|SetSpawnedModBox _returnvalue))
                        (EventDispatchers|BindEventtoOnDied _output_get (delegate Custom|ModBoxDied))
                        (CallFunction|UpdateModIndex)
                        (Variables|Default|SetSetHoloVisibility false)
                        (CallFunction|UpdateHoloBoxParameters true))) ; wtf -why do i need this to completely hotfix off the box from spawning
                  ))))
            (:then_1
              (Rendering|SetVisibility (Variables|Default|GetStaticMesh) false false))))))))

(event Custom|ModBoxDied
  (params (DamagedActor :type Actor%20Object%20Reference) (Damage :type Float%20%28single-precision%29) (InstigatedBy :type Controller%20Object%20Reference) (DamageCauser :type Actor%20Object%20Reference) (HitLocation :type Vector) (FHitComponent :type Primitive%20Component%20Object%20Reference) (BoneName :type Name) (Momentum :type Vector))
  ;[1 1 1] Spawn First ModBox
  ;;[0 0 0] Mod Box Died
  (Utilities|Time|SetTimerbyEvent (EventDispatchers|CreateEvent :!event Custom|SpawnModBox) (Variables|ModBox|GetCachedRespawnDelay))
  (bind _returnvalue (Utilities|Time|SetTimerbyEvent (EventDispatchers|CreateEvent :!event Custom|UpdateCountdownText) (Variables|Default|GetTimerUpdateFrequency) true)) ; Countdown Update
  (Variables|Default|SetCountdownTimer _returnvalue)
  (Variables|Default|SetSetHoloVisibility true)
  (CallFunction|UpdateHoloBoxParameters false))

(event Custom|RestartRespawnTimers
  ;[1 1 1] Spawn First ModBox
  (CallFunction|UpdateModIndex)
  (Utilities|Time|ClearandInvalidateTimerbyHandle (Variables|Default|GetCountdownTimer))
  (Variables|Default|SetRespawnTimeRemaining (Variables|ModBox|GetCachedRespawnDelay)) ; Reset Timer
  ;;[0 0 0] Mod Box Died
  (Variables|Default|SetSetHoloVisibility false)
  (Utilities|FlowControl|Delay 0.2) ; need this or bad things happen
  (Utilities|Time|SetTimerbyEvent (EventDispatchers|CreateEvent :!event Custom|SpawnModBox) (Variables|ModBox|GetCachedRespawnDelay))
  (bind _returnvalue (Utilities|Time|SetTimerbyEvent (EventDispatchers|CreateEvent :!event Custom|UpdateCountdownText) (Variables|Default|GetTimerUpdateFrequency) true)) ; Countdown Update
  (Variables|Default|SetCountdownTimer _returnvalue)
  (Variables|Default|SetSetHoloVisibility true)
  (CallFunction|UpdateHoloBoxParameters false))

(event Custom|UpdateModIndex
  (bind _modspawnindex (Variables|Default|GetModSpawnIndex))
  ;[0.9406 0 1] Update Mod Spawn Index
  (if (not (Variables|ModBox|GetRandomlySpawnMods))
    (Math|Integer|IncrementInt _modspawnindex
      (:%20%20
        (if (> _modspawnindex (- (Variables|Default|GetMaxModArrayCount) 1))
          (Variables|Default|SetModSpawnIndex 0) ; Reset to first index if the last Mod in the array was spawned
        )))))

(event Custom|UpdateCountdownText
  ;[0.08142 1 0] Update Cooldown Time
  (Variables|Default|SetRespawnTimeRemaining (- (Variables|Default|GetRespawnTimeRemaining) 1.0))
  (CallFunction|UpdateHoloBoxParameters false))

(event Custom|SpawnModBox
  ;[1 1 1] Spawn box, but first check if we overlap any actors
  (Variables|Default|SetAutoApplyMod false)
  (bind _outactors (Collision|BoxOverlapActors (+ (Transformation|GetActorLocation) "0.000000,0.000000,50.000000") "55.000000,55.000000,55.000000" (Variables|Default|GetObject%20Types) :!class /Script/Engine.KismetSystemLibrary))
  (for _array_element _outactors
    ;[0.1326 1 0] Are we overlapping anything?
    (bind _asfort_player_pawn (Utilities|Casting|CastToFortPlayerPawn _array_element)
      (:then
        (Pawn|Character|LaunchCharacterJump _asfort_player_pawn "0.000000,0.000000,1200.000000" false true true true))
      (:CastFailed
        (bind _asfort_athena_vehicle (Utilities|Casting|CastToFortAthenaVehicle _array_element))
        (Vehicle|Mod|ApplyVehicleMod _asfort_athena_vehicle (Class|GetClassDefaults (select (Variables|ModBox|GetRandomlySpawnMods) (Utilities|Array|RandomArrayItem (Variables|ModBox|GetModsToSpawn)) (Utilities|Array|Get(acopy) (Variables|ModBox|GetModsToSpawn) (Variables|Default|GetModSpawnIndex)))) 0)
        (Variables|Default|SetAutoApplyMod true)
        (CallFunction|RestartRespawnTimers)
        (CallFunction|ApplyModCosmetics))))
  ;[15 2.536 0] Spawn New Mod Box
  (if (not (Variables|Default|GetAutoApplyMod))
    (bind _returnvalue (Game|SpawnActorfromClass (select (Variables|ModBox|GetRandomlySpawnMods) (Utilities|Array|RandomArrayItem (Variables|ModBox|GetModsToSpawn)) (Utilities|Array|Get(acopy) (Variables|ModBox|GetModsToSpawn) (Variables|Default|GetModSpawnIndex))) (Transformation|GetActorTransform) "AlwaysSpawn" "SelectDefaultAtRuntime" self 0))
    (bind _asmod_box_parent (Utilities|Casting|CastToModBox_Parent _returnvalue))
    (bind _output_get (Variables|Default|SetSpawnedModBox _asmod_box_parent))
    (EventDispatchers|AssignOnDied _output_get (EventDispatchers|CreateEvent :!event Custom|ModBoxDied))
    (CallFunction|UpdateModIndex)
    (Utilities|Time|ClearandInvalidateTimerbyHandle (Variables|Default|GetCountdownTimer))
    (Variables|Default|SetRespawnTimeRemaining (Variables|ModBox|GetCachedRespawnDelay)) ; Reset Timer Text
    (Variables|Default|SetSetHoloVisibility false)
    (CallFunction|UpdateHoloBoxParameters true true)))

(event Custom|UpdateHoloBoxParameters
  (params (isInitialization :type Boolean) (isComplete :type Boolean))
  (bind _cachedrespawndelay (Variables|ModBox|GetCachedRespawnDelay))
  (bind _respawntimeremaining (Variables|Default|GetRespawnTimeRemaining))
  (bind _returnvalue (Math|Float|NormalizetoRange _respawntimeremaining 0.0 _cachedrespawndelay))
  (bind _cube (Variables|Default|GetCube))
  (bind _respawntimeremaining_2 (Variables|Default|GetRespawnTimeRemaining))
  (bind _cube_2 (Variables|Default|GetCube))
  ;[0.1156 0.2491 0.4] Holo Box FX
  (if isInitialization
    (Rendering|Material|SetScalarParameterValueonMaterials _cube "Mask Wipe Position" -8e+01)
    (Rendering|Material|SetScalarParameterValueonMaterials _cube "Blueprint Edge Brightness" 1.0)
    (if isComplete
      (Niagara|SpawnSystemAttached "/VehicleMods/FX/Systems/Generic/NS_VehicleMod_Spawn_Test.NS_VehicleMod_Spawn_Test" _cube "None" "0, 0, 0" "0, 0, 0" "KeepRelativeOffset" true)
      ;[0 0.2156 1] SOUND
      (Audio|SpawnSoundAttached "/VehicleMods/Gameplay/Actors/ModBox/Spawners/Audio/MSS_Modbox_Respawn.MSS_Modbox_Respawn" _cube))
    (else
      (Utilities|FlowControl|Sequence
        (:then_0
          (Rendering|Material|SetScalarParameterValueonMaterials _cube "Linear Interp" _returnvalue)
          (Rendering|Material|SetScalarParameterValueonMaterials _cube "Blueprint Edge Brightness" (Math|Float|Lerp 3.5 0.0 0))
          (Rendering|Material|SetScalarParameterValueonMaterials _cube_2 "Respawn Time Remaining" _respawntimeremaining_2)
          (if (== _respawntimeremaining_2 (Math|Float|Ceil (* (/ (Math|Float|Floor (* _returnvalue 9.0)) 9.0) 3e+01)))
            (Rendering|Material|SetScalarParameterValueonMaterials _cube_2 "Fade In" (Utilities|Time|GetGameTimeinSeconds))
            (Utilities|FlowControl|Delay 1.0)
            (Rendering|Material|SetScalarParameterValueonMaterials _cube_2 "Fade Out" (Utilities|Time|GetGameTimeinSeconds))))))))

(event Custom|ApplyModCosmetics
  (bind _returnvalue (Transformation|GetActorLocation))
  ;[1 1 1] Apply Mod Cosmetics
  (if (not (Networking|IsServer))
    (Niagara|SpawnSystematLocation (Variables|Default|GetBoxDestroyFX) _returnvalue)
    (Audio|SpawnSoundatLocation (Variables|Default|GetBoxDestroyedSound) _returnvalue)))

(fn OnRep_ModSpawnIndex)

(fn OnRep_SetHoloVisibility
  (Rendering|SetVisibility (Variables|Default|GetCube) (Variables|Default|GetSetHoloVisibility) true)
  (if (Variables|Default|GetSetHoloVisibility)
    (bind _mss_modbox_holo_lp (Variables|Default|GetMSS_Modbox_Holo_lp)
      (:then
        (Audio|Components|Audio|FadeIn _mss_modbox_holo_lp 0.0 1.0 0.0 "Linear" :!class /Script/Engine.AudioComponent))
      (:else))
    (else
      (bind _mss_modbox_holo_lp_2 (Variables|Default|GetMSS_Modbox_Holo_lp)
        (:then
          (Audio|Components|Audio|FadeOut _mss_modbox_holo_lp_2 0.2 0.0 "Linear" :!class /Script/Engine.AudioComponent))
        (:else)))))

(fn SetHotfixValues
  (bind _respawndelayrow (Variables|ModBox|GetRespawnDelayRow))
  (bind _returnvalue (ScalableFloat|GetValueatLevel _respawndelayrow))
  (Variables|ModBox|SetCachedRespawnDelay _returnvalue)
  (Variables|Default|SetRespawnTimeRemaining _returnvalue) ; Need to do this so the UI Updates properly
)

(fn ConstructionScript
  (|CallParentFunction)
  (Variables|Default|SetRespawnTimeRemaining (Variables|ModBox|GetCachedRespawnDelay)) ; Reset Timer Text
)
