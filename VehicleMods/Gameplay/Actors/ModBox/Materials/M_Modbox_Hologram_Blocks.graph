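# Translucent unlit hologram material for the VehicleMods ModBox, drawing glowing animated blocks with parameters for respawn timing.
# Local-position triplanar sampling of a grid-squares texture makes sine-animated cubes, alongside bounds-based edge glow and a virtual plane with its own cubes.
# A height mask with stepped interpolation and a respawn-timed layer flash are built as named reroutes.
# Color blending, depth fade, dithered opacity and optional normal-pulse WPO feed MF_CreativeMoveTool.
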Material /VehicleMods/Gameplay/Actors/ModBox/Materials/M_Modbox_Hologram_Blocks
graph {
 properties {
  bAutomaticallySetUsageInEditor=false
  bCastRayTracedShadows=false
  blendMode=BLEND_Translucent
  pixelDepthOffsetMode=PDOM_Legacy
  shadingModel=MSM_Unlit
  usedShadingModels=Unlit
 }
 nodes [
  n1 = Saturate(None=DotProduct(A=VertexNormalWS(), B=AtmosphericLightVector()))
  n2 = ScalarParameter() {defaultValue=0.25, parameterName=Sunlight Gleem}
  n3 = VectorParameter() {parameterName=Color Dark, defaultValue=<rgba 0.0149 0 0.4479 1>}
  n4 = VectorParameter() {parameterName=Color, defaultValue=<rgba 0 0.9324 1.5 1>}
  n5 = ScalarParameter() {defaultValue=0.5, parameterName=Darken Cubes}
  n6 = LinearInterpolate(
   A=Multiply(A=NamedRerouteUsage(), B=ScalarParameter() {parameterName=Param_3})
   B=Multiply(A=NamedRerouteUsage(), B=ScalarParameter() {defaultValue=1, parameterName=Param_4})
   Alpha=OneMinus(None=NamedRerouteUsage())
  )
  n7 = Power(Base=NamedRerouteUsage(), Exp=ScalarParameter() {defaultValue=1.5, parameterName=Param_8})
  n8 = LinearInterpolate(
   A=ScalarParameter() {defaultValue=2, parameterName=Param_7}
   B=ScalarParameter() {defaultValue=15, parameterName=Edge Glow Mult}
   Alpha=OneMinus(None=NamedRerouteUsage())
  )
  n9 = NamedRerouteUsage()
  n10 = Multiply(A=NamedRerouteUsage(), B=ScalarParameter() {defaultValue=5, parameterName=Param_18})
  n11 = Add(A=Add(A=@n9, B=NamedRerouteUsage()), B=@n10)
  n12 = Multiply(A=OneMinus(None=NamedRerouteUsage()), B=@n10)
  n13 = Add(
   A=Add(A=LinearInterpolate(A=@n5, Alpha=@n6), B=Multiply(A=@n7, B=@n8))
   B=Multiply(A=Add(A=@n11, B=@n12), B=ScalarParameter() {defaultValue=1, parameterName=Param_23})
  )
  n14 = DepthFade() {fadeDistanceDefault=10}
  n15 = Add(A=Add(A=Multiply(A=@n1, B=@n2), B=LinearInterpolate(A=@n3.RGB, B=@n4.RGB, Alpha=@n13)), B=Multiply(A=OneMinus(None=@n14)))
  n16 = NamedRerouteUsage()
  n17 = ScalarParameter() {defaultValue=0.25, parameterName=Static Cube Opacity}
  n18 = ScalarParameter() {defaultValue=0.62, parameterName=Param_6}
  n19 = NamedRerouteUsage()
  n20 = NamedRerouteUsage()
  n21 = LinearInterpolate(
   A=Multiply(A=NamedRerouteUsage())
   Alpha=LinearInterpolate(A=Multiply(A=@n16, B=@n17), B=LinearInterpolate(A=@n18, Alpha=@n19), Alpha=OneMinus(None=@n20))
  )
  n22 = Multiply(A=@n14, B=@n21)
  n23 = LinearInterpolate(
   A=MaterialFunctionCall(Alpha Threshold=@n21) {materialFunction=&DitherTemporalAA}.Result
   Alpha=@n21
  )
  n24 = NamedRerouteUsage()
  BreakMaterialAttributes!(Attr=MaterialFunctionCall(MA In=MakeMaterialAttributes!(
     BaseColor
     Metallic
     Specular
     Roughness
     Anisotropy
     EmissiveColor=@n15
     Opacity=@n22
     OpacityMask=@n23
     Normal
     Tangent
     WorldPositionOffset=@n24
     SubsurfaceColor
     ClearCoat
     ClearCoatRoughness
     AmbientOcclusion
     Refraction
     CustomizedUVs_0
     CustomizedUVs_1
     CustomizedUVs_2
     CustomizedUVs_3
     CustomizedUVs_4
     CustomizedUVs_5
     CustomizedUVs_6
     CustomizedUVs_7
     PixelDepthOffset
     ShadingModel
     Displacement
    )) {materialFunction=&MF_CreativeMoveTool}.MA Out)
  n25 = VectorParameter() {group=Derez, parameterName=DeRezSquare_TextureSize, defaultValue=<rgba 200 200 200 0>}
  n26 = Divide(A=LocalPosition() {includedOffsets=ExcludeOffsets}.XYZ, B=Multiply(A=Abs(None=@n25.RGB)) {constB=-1})
  ComponentMask(None=Normalize(VectorInput=@n26)) {g=true}
  Multiply(A=@n9, B=@n10)
  n27 = Constant()
  n28 = Constant() {r=1}
  n29 = Divide(A=NamedRerouteUsage(), B=Constant() {r=100}) {constB=55.0056}
  n30 = DotProduct(
   A=Normalize(VectorInput=MaterialFunctionCall(X=@n28, Y=@n29) {materialFunction=&MakeFloat2}.Result)
   B=Constant2Vector() {g=1}
  )
  n31 = ScalarParameter() {defaultValue=1.1, parameterName=zxcvzxc}
  n32 = Add(A=NamedRerouteUsage(), B=Multiply(A=Multiply(A=@n30, B=@n31) {constA=1, constB=0.4}, B=NamedRerouteUsage())) {constB=0}
  n33 = Multiply(
   A=ObjectLocalBounds().Extents
   B=MaterialFunctionCall() {materialFunction=&ObjectScale}.Scale XYZ
  )
  n34 = Multiply(A=NamedRerouteUsage(), B=@n31)
  n35 = MaterialFunctionCall(
   Plane Normal Axis=Transform(None=VectorParameter() {parameterName=Normal, defaultValue=<rgba 0 0 1 0>}.RGB) {transformSourceType=TRANSFORMSOURCE_Local}
   Plane X Axis=Transform(None=VectorParameter() {parameterName=X Axis, defaultValue=<rgba 1 0 0 0>}.RGB) {transformSourceType=TRANSFORMSOURCE_Local}
   Plane Center=TransformPosition(None=MaterialFunctionCall(X=@n27, Y=@n27, Z=@n32) {materialFunction=&MakeFloat3}.Result) {transformSourceType=TRANSFORMPOSSOURCE_Instance, transformType=TRANSFORMPOSSOURCE_World}
   UV Scale=Multiply(A=Add(A=@n33, B=@n34), B=ScalarParameter() {defaultValue=2, parameterName=Bound UV Scaling Mult})
  ) {materialFunction=&VirtualPlaneCoordinates}
  n36 = NamedRerouteUsage()
  n37 = DotProduct(A=Constant2Vector() {g=0.5, r=0.5}, B=Constant2Vector() {g=1})
  n38 = Divide(
   A=Transform(None=@n35.World Position - Centered) {transformSourceType=TRANSFORMSOURCE_World, transformType=TRANSFORM_Local}
   B=Add(A=ObjectLocalBounds().Half Extents, B=Multiply(A=@n36, B=@n37) {constB=0.5})
  )
  n39 = Saturate(None=OneMinus(None=Abs(None=@n38)))
  n40 = SphereMask(A=Ceil(None=@n39), B=ScalarParameter() {defaultValue=1, parameterName=Param_9}, Radius=ScalarParameter() {parameterName=Param_10}) {hardnessPercent=0.5}
  n41 = SphereMask(A=@n39, B=Constant() {r=1}) {attenuationRadius=1.2, hardnessPercent=75}
  Multiply(A=@n40, B=@n41)
  n42 = TextureSample(UVs=ComponentMask(None=@n26) {b=true, r=true}) {
   samplerSource=SSM_Wrap_WorldGroupSettings
   samplerType=SAMPLERTYPE_LinearColor
   texture=/Game/Packages/Fortress_SharedMaterials/Textures/T_Slowfield_GridSquares_01
  }
  n43 = TextureSample(UVs=ComponentMask(None=@n26) {b=true, g=true}) {
   samplerSource=SSM_Wrap_WorldGroupSettings
   samplerType=SAMPLERTYPE_LinearColor
   texture=&T_Slowfield_GridSquares_01
  }
  n44 = VertexInterpolator(VS=PreSkinnedNormal())
  n45 = Constant() {r=1}
  n46 = MaterialFunctionCall(In=Abs(None=ComponentMask(None=@n44.PS) {r=true}), Contrast=@n45) {materialFunction=&CheapContrast}
  n47 = TextureSample(UVs=ComponentMask(None=@n26) {g=true, r=true}) {
   samplerSource=SSM_Wrap_WorldGroupSettings
   samplerType=SAMPLERTYPE_LinearColor
   texture=&T_Slowfield_GridSquares_01
  }
  n48 = Abs(None=ComponentMask(None=@n44.PS) {b=true})
  n49 = LinearInterpolate(
   A=LinearInterpolate(A=@n42.RGB, B=@n43.RGB, Alpha=ComponentMask(None=@n46.Result) {r=true})
   B=@n47.RGB
   Alpha=ComponentMask(None=MaterialFunctionCall(In=@n48, Contrast=@n45) {materialFunction=&CheapContrast}.Result) {r=true}
  )
  n50 = Time()
  n51 = ScalarParameter() {defaultValue=0.5, parameterName=Time Scale}
  n52 = Sine(None=Add(A=@n49, B=Multiply(A=@n50, B=@n51)))
  NamedRerouteDeclaration(None=Multiply(A=ConstantBiasScale(None=@n52), B=@n49)) {name=Animated Cubes, nodeColor=<rgba 0 1 0.4609 1>}
  NamedRerouteDeclaration(None=@n49) {name=Static Cubes, nodeColor=<rgba 1 0.1406 0 1>}
  n53 = ScalarParameter() {defaultValue=-2.5, parameterName=s}
  n54 = ScalarParameter() {defaultValue=-5, parameterName=Param_2}
  n55 = NamedRerouteUsage()
  n56 = ScalarParameter() {defaultValue=5, parameterName=Mult}
  n57 = NamedRerouteUsage()
  n58 = Time()
  n59 = ScalarParameter() {parameterName=Start Time}
  n60 = ConstantBiasScale(None=Cosine(None=Subtract(A=@n58, B=@n59)))
  n61 = Add(
   A=Add(A=LinearInterpolate(A=@n53, B=@n54, Alpha=@n55), B=LinearInterpolate(B=@n56, Alpha=@n57))
   B=Multiply(A=OneMinus(None=@n60), B=ScalarParameter() {defaultValue=-10, parameterName=Idle WPO})
  )
  n62 = Constant()
  NamedRerouteDeclaration(None=StaticSwitchParameter(True=@n61, False=@n62) {defaultValue=true, parameterName=UseWPO}) {name=WPO Pulse, nodeColor=<rgba 0.6172 1 0 1>}
  n63 = ScalarParameter() {defaultValue=62.5, parameterName=Box End}
  n64 = ScalarParameter() {defaultValue=-50, parameterName=Box Start}
  n65 = Abs(None=@n64)
  NamedRerouteDeclaration(None=Divide(A=Add(A=@n63, B=@n65))) {name=Divisor, nodeColor=<rgba 0.2188 0 1 1>}
  n66 = Time() {period=30}
  n67 = ScalarParameter() {defaultValue=1}
  n68 = Saturate(None=MaterialFunctionCall(
    TimestampA=ScalarParameter() {parameterName=Fade In}
    TransitionLength_In=ScalarParameter() {defaultValue=1, parameterName="TransitionLength_In (S)"}
    TimestampB=ScalarParameter() {defaultValue=1, parameterName=Fade Out}
    TransitionLength_Out=ScalarParameter() {defaultValue=1, parameterName="TransitionLength_Out (S)"}
    Time=Multiply(A=@n66, B=@n67) {constB=3}
    Transition In Only?=StaticBoolParameter() {parameterName="Transition In Only? (SB)"}
   ) {materialFunction=&MF_Timestamp_InOut_01}.SmoothStep)
  n69 = ScalarParameter() {defaultValue=8, parameterName=Param_1}
  n70 = ScalarParameter() {defaultValue=30, parameterName=Respawn Time Max, sliderMax=999}
  n71 = OneMinus(None=Divide(A=ScalarParameter() {defaultValue=4.9894, parameterName=Respawn Time Remaining}, B=@n70) {constB=30})
  n72 = Sine(None=Divide(A=Multiply(A=@n71, B=@n70) {constB=30}) {constB=3.3333})
  NamedRerouteDeclaration(None=LinearInterpolate(B=Power(Base=@n68, Exp=@n69), Alpha=Saturate(None=@n72))) {name=Layer Flash, nodeColor=<rgba 0 0.7187 1 1>}
  n73 = NamedRerouteUsage()
  n74 = ScalarParameter() {parameterName=Edge Distortion}
  n75 = Subtract(A=Add(A=Multiply(A=@n73, B=@n74), B=LocalPosition() {includedOffsets=ExcludeOffsets}.Z), B=NamedRerouteUsage())
  n76 = ScalarParameter() {defaultValue=1, parameterName=Mask Soften}
  NamedRerouteDeclaration(None=Saturate(None=Divide(A=@n75, B=@n76))) {name=Mask, nodeColor=<rgba 0 0.1797 1 1>}
  n77 = ScalarParameter() {defaultValue=-0.12, parameterName=Edge Glow Contrast}
  n78 = MaterialFunctionCall() {materialFunction=&BoundingBoxBased_0-1_UVW}
  n79 = Add(A=Multiply(A=@n78.R, B=OneMinus(None=@n78.R)), B=Multiply(A=@n78.G, B=OneMinus(None=@n78.G)))
  n80 = Multiply(A=@n78.B, B=OneMinus(None=@n78.B))
  n81 = Power(Base=Saturate(None=Add(A=@n79, B=@n80)), Exp=ScalarParameter() {defaultValue=0.4, parameterName=Edge Glow Power})
  NamedRerouteDeclaration(None=Power(Base=LinearInterpolate(B=@n77, Alpha=@n81) {constA=1, constB=-2}) {constExponent=3}) {name=Edge Glow, nodeColor=<rgba 0.5234 0 1 1>}
  n82 = NamedRerouteUsage()
  n83 = DotProduct(A=CameraVectorWS(), B=Transform(None=Constant3Vector() {constant=<rgba 0 0 1 0>}) {transformSourceType=TRANSFORMSOURCE_Local})
  n84 = Multiply(
   A=LinearInterpolate(A=ScalarParameter() {defaultValue=0.35, parameterName=Virtual Plane Cube Contrast}, Alpha=NamedRerouteUsage())
   B=Ceil(None=Saturate(None=@n83))
  )
  NamedRerouteDeclaration(None=Multiply(A=@n40, B=Multiply(A=@n82, B=@n84))) {name=Virual Plane, nodeColor=<rgba 0.4531 1 0 1>}
  n85 = OneMinus(None=@n41)
  n86 = ScalarParameter() {defaultValue=3, parameterName=Param_11}
  NamedRerouteDeclaration(None=Multiply(A=@n40, B=Power(Base=@n85, Exp=@n86) {constExponent=20})) {name=Virual Plane Glow, nodeColor=<rgba 0 1 0.3203 1>}
  NamedRerouteDeclaration(None=@n35.UVs) {name=Virtual Plane UVs, nodeColor=<rgba 0.125 0 1 1>}
  n87 = VertexNormalWS()
  NamedRerouteDeclaration(None=StaticSwitchParameter(True=Multiply(A=@n87, B=@n61), False=@n62) {defaultValue=true, parameterName=UseWPO}) {name=WPO, nodeColor=<rgba 0.3828 0 1 1>}
  n88 = TextureSample(UVs=NamedRerouteUsage()) {
   samplerSource=SSM_Wrap_WorldGroupSettings
   samplerType=SAMPLERTYPE_LinearColor
   texture=&T_Slowfield_GridSquares_01
  }
  n89 = Time()
  n90 = ScalarParameter() {defaultValue=0.5, parameterName=Time Scale}
  n91 = Sine(None=Add(A=@n88.RGB, B=Multiply(A=@n89, B=@n90)))
  NamedRerouteDeclaration(None=Multiply(A=ConstantBiasScale(None=@n91), B=@n88.RGB)) {name=Virual Plane Animated Cubes, nodeColor=<rgba 1 0 0.5156 1>}
  n92 = Clamp(None=ScalarParameter() {defaultValue=0.3, parameterName=Linear Interp})
  n93 = ScalarParameter() {defaultValue=9, parameterName=Amount to Step}
  n94 = Floor(None=Multiply(A=NamedRerouteDeclaration(None=@n92) {name=Time, nodeColor=<rgba 0.7109 0 1 1>}, B=@n93))
  NamedRerouteDeclaration(None=LinearInterpolate(A=@n63, B=@n64, Alpha=Divide(A=@n94, B=@n93)) {constA=70, constB=-56}) {name=Interp, nodeColor=<rgba 0 1 0.7891 1>}
  NamedRerouteUsage()
  NamedRerouteUsage()
  NamedRerouteUsage()
  n95 = Time()
  OneMinus(None=Modulo!(A=Multiply(A=@n95) {constB=0.1}, B=Constant() {r=1}))
  VertexInterpolator()
  VertexInterpolator()
  VertexNormalWS()
  WorldPosition() {worldPositionShaderOffset=WPT_ExcludeAllShaderOffsets}
 ]
}
