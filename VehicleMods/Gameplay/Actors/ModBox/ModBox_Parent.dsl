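; Base mod box actor: a crate that applies a tagged mod to a compatible vehicle that destroys it.
; Vehicle overlap triggers server-side Die; OnDied applies the mod via the damage-causing vehicle, then runs destruction cosmetics.
; RepNotifies handle mesh hiding, audio bus mix and customizable dynamic material.
; Interaction overrides adjust prompts and disable the key unless the player occupies a moddable vehicle; setters fill the world UI label.
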(event EventBeginPlay
  (EventDispatchers|AssignOnDied (delegate Custom|OnDied_Event))
  (EventDispatchers|AssignOnDeathPlayEffectsDelegate (delegate Custom|OnDeathPlayEffectsDelegate_Event)))

(event Custom|OnDeathPlayEffectsDelegate_Event
  (params (Damage :type Float%20%28single-precision%29) (DamageTags :type Gameplay%20Tag%20Container%20Structure%20%28by%20ref%29) (Momentum :type Vector) (HitInfo :type Hit%20Result%20Structure%20%28by%20ref%29) (InstigatedBy :type Fort%20Pawn%20Object%20Reference) (DamageCauser :type Actor%20Object%20Reference) (EffectContext :type Gameplay%20Effect%20Context%20Handle%20Structure))
  (bind (_bblockinghit _binitialoverlap _time _distance _location _impactpoint _normal _impactnormal _physmat _hitactor) (Collision|BreakHitResult HitInfo))
  (Audio|SpawnSoundatLocation (Variables|Default|GetBoxFinalImpactSound) (Transformation|GetActorLocation _hitactor)))

(event Custom|DestroyCosmetics
  (bind _returnvalue (Transformation|GetActorLocation))
  ;[0.5521 0.5749 1] Destruction Cosmetics!
  (if (Networking|IsServer)
    (Variables|Default|SetbHideModBox true) ; OnRep minimizes delay between clients so the crate doesn't remain visually
    (else
      (Niagara|SpawnSystematLocation (Variables|ModBoxFX|GetBoxDestroyFX) _returnvalue)
      (Audio|SpawnSoundatLocation (Variables|Default|GetBoxDestroyedSound) _returnvalue))))

(event Custom|OnDied_Event
  (params (DamagedActor :type Actor%20Object%20Reference) (Damage :type Float%20%28single-precision%29) (InstigatedBy :type Controller%20Object%20Reference) (DamageCauser :type Actor%20Object%20Reference) (HitLocation :type Vector) (FHitComponent :type Primitive%20Component%20Object%20Reference) (BoneName :type Name) (Momentum :type Vector))
  (bind _modtagtoapply (Variables|ModTag|GetModTagToApply))
  ;[1 1 1] When Box dies, apply Mod
  (Variables|Default|SetHit%20Location HitLocation)
  (bind _asfort_athena_vehicle (Utilities|Casting|CastToFortAthenaVehicle DamageCauser)
    (:then
      (if (Vehicle|Mod|CanApplyMod _asfort_athena_vehicle _modtagtoapply)
        (if (Variables|Default|GetbAllowApplyMod)
          (Vehicle|Mod|ApplyVehicleMod _asfort_athena_vehicle _modtagtoapply (Fortnite|GetFortPlayerControllerfromActor (Vehicle|GetDriver _asfort_athena_vehicle)))
          (else
            _))
        (Variables|Default|SetbShouldApplyMix true)
        (else
          (Variables|Default|SetbShouldApplyMix false)))
      (CallFunction|DestroyCosmetics) ; Play Destroyed FX
    )
    (:CastFailed
      _)))

(event Custom|DeactivateCBM
  (bind _controlbusmix (Variables|Default|GetControlBusMix))
  (Utilities|IsValid _controlbusmix
    (:Is%20Valid
      (Audio|Modulation|DeactivateControlBusMix _controlbusmix))))

(event OnComponentBeginOverlap(VehicleOverlapBox)
  (params (OverlappedComponent :type Primitive%20Component%20Object%20Reference) (OtherActor :type Actor%20Object%20Reference) (OtherComp :type Primitive%20Component%20Object%20Reference) (OtherBodyIndex :type Integer) (bFromSweep :type Boolean) (SweepResult :type Hit%20Result%20Structure%20%28by%20ref%29))
  ;[1 1 1] If a Vehicle collides with the Box then DIE
  (bind _asfort_athena_vehicle (Utilities|Casting|CastToFortAthenaVehicle OtherActor))
  (if (not (Class|FortAthenaVehicle|HasExploded _asfort_athena_vehicle))
    (if (Networking|IsServer)
      (Gameplay|Die :Killer (Fortnite|GetFortPlayerControllerfromActor (Vehicle|GetDriver _asfort_athena_vehicle)) :DamageCauser OtherActor))))

(fn BlueprintDisplaySubInteractIcon
  (params (InteractingPawn :type Fort%20Pawn%20Object%20Reference) (InteractionBeingAttempted :type EInteractionBeingAttempted%20Enum :default "FirstInteraction") (InteractionType :type TInteractionType%20Enum :default "IT_NoInteraction"))
  (if (not (WorldUI|IsPlayerinModdableVehicle InteractingPawn))
    (return true)))

(fn BlueprintGetFailedInteractionString
  (params (InteractingPawn :type Fort%20Pawn%20Object%20Reference) (InteractionBeingAttempted :type EInteractionBeingAttempted%20Enum :default "FirstInteraction"))
  (if (WorldUI|IsPlayerinModdableVehicle InteractingPawn)
    (bind _returnvalue (|CallParentFunction InteractingPawn InteractionBeingAttempted))
    (return _returnvalue)
    (elif (Variables|Default|GetbShowFailedInteractionModBoxName)
      (return (Utilities|String|SelectText (Variables|WorldUI|GetModBoxName) (Variables|WorldUI|GetModBoxName_Override) (Utilities|Text|TextIsEmpty (Variables|WorldUI|GetModBoxName_Override))))
      (else
        (return "")))))

(fn BlueprintGetFailedSubInteractionString
  (params (InteractingPawn :type Fort%20Pawn%20Object%20Reference) (InteractionBeingAttempted :type EInteractionBeingAttempted%20Enum :default "FirstInteraction"))
  (if (WorldUI|IsPlayerinModdableVehicle InteractingPawn)
    (bind _returnvalue (|CallParentFunction InteractingPawn InteractionBeingAttempted))
    (return _returnvalue)
    (else
      (return (Variables|WorldUI|GetFailedSubInteraction)))))

(fn BlueprintGetFailedSubTextTexture
  (params (InteractingPawn :type Fort%20Pawn%20Object%20Reference) (InteractionBeingAttempted :type EInteractionBeingAttempted%20Enum :default "FirstInteraction"))
  (if (not (WorldUI|IsPlayerinModdableVehicle InteractingPawn))
    (return (LocalVariables|GetFailedInteractionTexture))))

(fn BlueprintUpdateDynamicInteractionOptions
  (params (InteractingPawn :type Fort%20Pawn%20Object%20Reference) (InteractionBeingAttempted :type EInteractionBeingAttempted%20Enum :default "FirstInteraction") (InteractionType :type TInteractionType%20Enum :default "IT_NoInteraction"))
  (bind _outgameplaytags (|CallParentFunction InteractingPawn InteractionBeingAttempted InteractionType))
  (LocalVariables|SetInteractionOptions _outgameplaytags)
  (if (not (WorldUI|IsPlayerinModdableVehicle InteractingPawn))
    (GameplayTags|AddGameplayTag (LocalVariables|GetInteractionOptions) "(TagName=\"HUD.Athena.InteractionOptions.DisableInteractKey\")" :!class /Script/GameplayTags.BlueprintGameplayTagLibrary))
  (return (LocalVariables|GetInteractionOptions)))

(fn IsPlayerInModdableVehicle
  (params (Player :type Pawn%20Object%20Reference))
  (bind _vehicle_occupied (Utilities|IsinVehicle Player)
    (:true
      (return (Vehicle|Mod|CanApplyMod _vehicle_occupied (Variables|WorldUI|GetTestModTag))))
    (:false
      (return false))))

(fn OnRep_bHideModBox
  (Rendering|SetVisibility (Variables|Default|GetStaticMesh) (not (Variables|Default|GetbHideModBox)) true))

(fn OnRep_bShouldApplyMix
  (if (Variables|Default|GetbShouldApplyMix)
    (Audio|Modulation|ActivateControlBusMix (Variables|Default|GetControlBusMix))
    (Utilities|Time|SetTimerbyFunctionName 0 "DeactivateCBM" 1.0)))

(fn OnRep_CustomMatParams_Struct
  (bind (_custommatparams_struct_modbox_customizablematerial_mi_15_6df26ff6488d80c49191958d017f095f _custommatparams_struct_beditparameters_17_dee0c7774e58b9d685777e8b775411ca _custommatparams_struct_logotexture_12_944e265f41970b63b29b0893d5a02865 _custommatparams_struct_modbox_logo_color_10_1a62809c4016c750cafbcfa63b9f03ac _custommatparams_struct_modbox_logoscale_8_3b298913439b57677f4156846c2dfca5 _custommatparams_struct_modbox_primary_color_11_4a9d89e04c05f2882e25c480d5a11125) (Utilities|Struct|BreakVehicleModCustomizableParamsStruct (Variables|GetCustomMatParams_Struct_ModBox_CustomizableMaterial_MI_15_6DF26FF6488D80C49191958D017F095F)))
  (bind _custommat_mid (LocalVariables|GetCustomMat_MID))
  (Networking|CanExecuteCosmeticEvents
    (:True
      (LocalVariables|SetCustomMat_MID (Material|GetMIDfromMaterial (Variables|Default|GetStaticMesh) 0 _custommatparams_struct_modbox_customizablematerial_mi_15_6df26ff6488d80c49191958d017f095f))
      (if _custommatparams_struct_beditparameters_17_dee0c7774e58b9d685777e8b775411ca
        (Rendering|Material|SetTextureParameterValue _custommat_mid "LogoTexture" _custommatparams_struct_logotexture_12_944e265f41970b63b29b0893d5a02865)
        (Rendering|Material|SetVectorParameterValue _custommat_mid "ModBox_Logo_Color" _custommatparams_struct_modbox_logo_color_10_1a62809c4016c750cafbcfa63b9f03ac :!class /Script/Engine.MaterialInstanceDynamic)
        (Rendering|Material|SetScalarParameterValue _custommat_mid "ModBox_LogoScale" _custommatparams_struct_modbox_logoscale_8_3b298913439b57677f4156846c2dfca5 :!class /Script/Engine.MaterialInstanceDynamic)
        (Rendering|Material|SetVectorParameterValue _custommat_mid "ModBox_Primary_Color" _custommatparams_struct_modbox_primary_color_11_4a9d89e04c05f2882e25c480d5a11125 :!class /Script/Engine.MaterialInstanceDynamic)
        (Rendering|Material|SetMaterial (Variables|Default|GetStaticMesh) 0 _custommat_mid)
        (else
          (Rendering|Material|SetMaterial (Variables|Default|GetStaticMesh) 0 (LocalVariables|GetCustomMat_MID)))))))

(fn OnRep_Destroyed
  (Niagara|SpawnSystematLocation :SystemTemplate (Variables|ModBoxFX|GetBoxDestroyFX) :Location (Transformation|GetActorLocation) :Scale "8.000000,8.000000,8.000000" :bAutoDestroy true :bAutoActivate true :PoolingMethod "None" :bPreCullCheck true)
  (Development|PrintString "DIEDDDDD FX" true true "(R=0.000000,G=0.660000,B=1.000000,A=1.000000)" 9.0))

(fn SetLabelText
  (params (LabelTitle :type Text) (LabelDesc :type Text) (LabelImage :type Texture%202D%20Soft%20Object%20Reference))
  (Variables|WorldUI|SetModBoxName LabelTitle)
  (Networking|CanExecuteCosmeticEvents
    (:True
      (bind _aswbp_mod_box_world_ui (Utilities|Casting|CastToWBP_ModBox_WorldUI (UserInterface|GetWidget (Variables|Default|GetFortWidget))))
      (Class|WBPWorldUILabel|EventSetWorldUILabelInfo _aswbp_mod_box_world_ui (Utilities|String|SelectText (Variables|WorldUI|GetModBoxName) (Variables|WorldUI|GetModBoxName_Override) (Utilities|Text|TextIsEmpty (Variables|WorldUI|GetModBoxName_Override))) (Variables|GetLabelDesc))
      (Class|WBPWorldUILabel|EventSetWorldUILabelImage _aswbp_mod_box_world_ui (Variables|GetLabelImage))
      (Class|WBPModBoxWorldUI|SetIsSpawner _aswbp_mod_box_world_ui false)
      (Variables|/VehicleMods/Gameplay/Actors/ModBox/UI/WBP_ModBox_WorldUI.WBP_ModBox_WorldUI_C|SetTestModTag (Variables|ModTag|GetModTagToApply) _aswbp_mod_box_world_ui)
      (Rendering|SetVisibility (Variables|Default|GetFortWidget) true false))))

(fn SetLabelText_TitleOverride
  (params (LabelTitleOverride :type Text))
  (Variables|WorldUI|SetModBoxName_Override LabelTitleOverride)
  (Networking|CanExecuteCosmeticEvents
    (:True
      (bind _aswbp_mod_box_world_ui (Utilities|Casting|CastToWBP_ModBox_WorldUI (UserInterface|GetWidget (Variables|Default|GetFortWidget))))
      (Class|WBPWorldUILabel|EventSetWorldUILabelInfo _aswbp_mod_box_world_ui (Variables|WorldUI|GetModBoxName_Override) (Variables|/Game/Athena/HUD/WorldUI/WBP_WorldUI_Label.WBP_WorldUI_Label_C|GetSupportingTitle _aswbp_mod_box_world_ui)))))

(fn ConstructionScript)
